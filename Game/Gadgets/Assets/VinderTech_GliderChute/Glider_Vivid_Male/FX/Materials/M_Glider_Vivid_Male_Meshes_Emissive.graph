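# Material for Niagara mesh particles of the Vivid Male glider, giving its FX meshes a particle-tinted emissive look.
# The glider's diffuse texture is multiplied by particle color and passed to the Emissive input of MF_FX_EmissiveAdjustments.
# An Emissive scalar parameter and the normal and mask texture samples are present, with no connections shown.
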Material /Game/Gadgets/Assets/VinderTech_GliderChute/Glider_Vivid_Male/FX/Materials/M_Glider_Vivid_Male_Meshes_Emissive
graph {
 properties {bUsedWithNiagaraMeshParticles=true, pixelDepthOffsetMode=PDOM_Legacy, usedShadingModels=Default Lit}
 nodes [
  MaterialFunctionCall(Emissive=Multiply(
    A=TextureSample() {texture=/Game/Gadgets/Assets/VinderTech_GliderChute/Glider_Vivid_Male/Textures/T_Glider_Vivid_Male_D}.RGB
    B=ParticleColor().RGB
   )) {materialFunction=&MF_FX_EmissiveAdjustments}
  ScalarParameter() {defaultValue=0.3, parameterName=Emissive}
  TextureSample() {samplerType=SAMPLERTYPE_Normal, texture=/Game/Gadgets/Assets/VinderTech_GliderChute/Glider_Vivid_Male/Textures/T_Glider_Vivid_Male_N}
  TextureSample() {samplerType=SAMPLERTYPE_Masks, texture=/Game/Gadgets/Assets/VinderTech_GliderChute/Glider_Vivid_Male/Textures/T_Glider_Vivid_Male_S}
 ]
}
